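# Niagara effect of a glowing bug that fades in, flies a winding path and homes in on the User.RocketPos position.
# Bug_Wings is the leader: chained timelines drive a twelve-point spline, then a blend to the target, orienting the mesh along travel and ramping material animation and light.
# Body, corona, flare, dark gradient, turbulent ember and ribbon trail emitters follow the leader through particle attribute readers.
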NiagaraSystem /Figment_S04_Geode/Athena/Effects/Particles/Geode/NS_Geode_AlteredBug
dependencies {
 usedDataInterfaces [
  &NiagaraDataInterfaceCurve
  &NiagaraDataInterfaceParticleRead
  /Script/Niagara.NiagaraDataInterfaceVectorField
  &NiagaraDataInterfaceVectorCurve
 ]
 usedDynamicInputs []
 usedModules [
  &self:SystemSpawnScript.SystemScriptSource.SystemScriptGraph.NiagaraNodeAssignment_0.SetVariables_#1
  &SystemState
 ]
}
emitters {
 Bug_Body {
  emitterUpdateScript [
   &EmitterState(
    Delay First Loop Only=true
    Inactive Response=::NewEnumerator1
    Life Cycle Mode=::NewEnumerator1
    Loop Delay=$System.EmitterDelay
    Loop Duration=0.5
    UseLoopDelay=true
   )
   &SpawnBurst_Instantaneous(Use Spawn Probability=false, UseLoopCountLimit=false)
   &SpawnBurst_Instantaneous(Spawn Time=0.1, Use Spawn Probability=false, UseLoopCountLimit=false) {module=SpawnBurst_Instantaneous001}
   &SpawnBurst_Instantaneous(Spawn Time=0.2, Use Spawn Probability=false, UseLoopCountLimit=false) {module=SpawnBurst_Instantaneous002}
  ]
  particleSpawnScript [
   &InitializeParticle(
    Color=<rgba 3 0.3 0.345 1>
    Lifetime=1.5
    Sprite Size Mode=::NewEnumerator3
    Uniform Sprite Size=25
    UsePositionOffset=false
    Write Material Random=true
   )
   &DynamicMaterialParameters(
    Index 0 Param 1=&RandomRangeFloat(Minimum=0.4, Maximum=0.6)
    Index 0 Param 2=1
    Index 0 Param 3=1
    Module.Index0Float0Write=true
    Module.Index0Float1Write=true
    Module.Index0Float2Write=true
    Module.Index0Float3Write=false
   )
  ]
  particleUpdateScript [
   &SolveForcesAndVelocity(Clamp Velocity=false, Limit Acceleration=false)
   &ParticleState()
   &ScaleColor(
    Scale Alpha=&FloatFromCurve(FloatCurve=DataInterface() {propertyValues {Curve {
        defaultValue=1
        keys [
         {interpMode=RCIM_Cubic, tangentMode=RCTM_User, time=0, value=0}
         {interpMode=RCIM_Cubic, tangentMode=RCTM_User, time=0.6, value=1}
         {tangentMode=RCTM_User, time=1, value=0}
        ]
       }}})
    ScaleA=true
    ScaleRGB=true
   )
   &self:Follow_Pos_BugBody(
    Offset Rotation:Quat4f=/CascadeToNiagaraConverter/NiagaraScripts/DynamicInputScripts/MakeQuatFromEuler(Euler:Vector3f=<0 36 0>, Scale:NiagaraFloat=1)
    Offsets:NiagaraPosition=&RandomRangeVector(Minimum=<0 0 -10>, Maximum=<0 0 5>)
    Particle Attribute Reader=DataInterface() {propertyValues {EmitterBinding {bindingMode=Other, emitterName=Bug_Wings}}}
   ) {module=Follow_Pos_BugBody}
  ]
  rendererClasses [&NiagaraSpriteRendererProperties]
  renderers [{rendererClass=&NiagaraSpriteRendererProperties}]
 }
 Bug_Wings {
  emitterUpdateScript [
   &EmitterState()
   &SpawnBurst_Instantaneous(Use Spawn Probability=false, UseLoopCountLimit=true)
  ]
  particleSpawnScript [
   &InitializeParticle(Color Mode=::NewEnumerator0, Lifetime=5, Mesh Scale Mode=::NewEnumerator3, UsePositionOffset=false, Write Material Random=true)
   &InitialMeshOrientation(Mesh Orientation Mode=::NewEnumerator2, Orientation Vector=<-1 0 0>, Rotation=<0 -0.1 0>, Use Rotation Vector=true)
  ]
  particleUpdateScript [
   &ParticleState(Kill Particles When Lifetime Has Elapsed=false, Let Infinitely Lived Particles Die When Emitter Deactivates=true)
   &ScaleMeshSize(Scale Factor=&VectorFromFloat(Value=&FloatFromCurve(FloatCurve=DataInterface() {propertyValues {Curve {keys [<at 0 0.15>,<at 1 0.075>]}}}, CurveIndex=$System.Timeline.NormalizedTime)))
   &ScaleColor(
    Scale Alpha=&FloatFromCurve(
     FloatCurve=DataInterface() {propertyValues {Curve {defaultValue=1, keys [<at 0 0 RCIM_Cubic>,<at 0.6 1 RCIM_Cubic>]}}}
     CurveIndex=$System.Timeline.NormalizedTime
    )
    ScaleA=true
    ScaleRGB=false
   )
   &DynamicMaterialParameters(
    Index 0 Param 1=0.175
    Index 0 Param 4=&Waveform(
     "Waveform [1]"=::NewEnumerator7
     "[1] Frequency"=&FloatFromCurve(
      FloatCurve=DataInterface() {propertyValues {Curve {
         defaultValue=1
         keys [
          {arriveTangent=1.9077, interpMode=RCIM_Cubic, time=0.5, value=0}
          {arriveTangent=3.6053, interpMode=RCIM_Cubic, leaveTangent=3.6053, tangentMode=RCTM_User, time=1, value=1}
         ]
        }}}
      CurveIndex=$System.Timeline.NormalizedTime
      Scale Curve=&Lerp_Float(A=3, B=4, Alpha=&Waveform("Waveform [1]"=::NewEnumerator12, "[1] Frequency"=1, "[1] Amplitude Min/Max"=<0 1>))
     )
     "[1] Amplitude Min/Max"=<0 1>
    )
    Index 1 Param 1=3000
    Index 1 Param 3=&Lerp_Float(A=1, B=0, Alpha=$System.Timeline.NormalizedTime)
    Module.Index0Float0Write=true
    Module.Index0Float1Write=true
    Module.Index0Float2Write=true
    Module.Index0Float3Write=true
    Module.Index1Float0Write=true
    Module.Index1Float1Write=false
    Module.Index1Float2Write=true
    Module.Index1Float3Write=false
    Write Parameter Index 1=::NewEnumerator1
   )
   &Timeline(
    Play=&SetBoolByFloatComparison(A=$System.Timeline.NormalizedTime, B=0.995, Comparison Type=::NewEnumerator2)
    Play Rate=0.0833
   )
   &Timeline(Play=&SetBoolByFloatComparison(A=$Particles.Timeline.NormalizedTime, B=1, Comparison Type=::NewEnumerator2)) {module=Timeline001}
   /SetVariables(Particles.PreviousPosition:NiagaraPosition=$Particles.Position) {isSetParameters=true}
   &self:SplineThroughMultiplePoints(
    Pos 01:Vector3f=<200 0 125>
    Pos 02:Vector3f=<0 150 100>
    Pos 03:Vector3f=<-100 0 100>
    Pos 04:Vector3f=<0 -100 75>
    Pos 05:Vector3f=<100 0 75>
    Pos 06:Vector3f=<0 100 50>
    Pos 07:Vector3f=<-75 0 50>
    Pos 08:Vector3f=<0 -75 50>
    Pos 09:Vector3f=<75 0 75>
    Pos 10:Vector3f=<0 100 100>
    Pos 11:Vector3f=<-100 0 100>
    Pos 12:Vector3f=<50 -150 100>
    T:NiagaraFloat=$Particles.Timeline.NormalizedTime
   ) {module=SplineThroughMultiplePoints}
   /SetVariables(Particles.Position=&LerpPosition(
     A=$Particles.Position
     B=&TransformPosition(Position=$User.RocketPos, Source Space=::World, Destination Space=::Simulation)
     Alpha=$Particles.Timeline001.NormalizedTime
    )) {isSetParameters=true}
   /SetVariables(Particles.Direction:Vector3f=&RotateVectorByQuaternion(
     Vector To Rotate=&Lerp_Vector(
      A=<-1 0 0>
      B=&NormalizeVector(
       Vector To Normalize=&Subtract_Position(A=$Particles.Position, B=$Particles.PreviousPosition)
       Safe Normalize=true
       Safe Fallback Threshold=0
       Safe Fallback Vector=<0 0 0.001>
      )
      Alpha=$System.Timeline.NormalizedTime
     )
     Quaternion=$Engine.Owner.Rotation
    )) {isSetParameters=true}
   &UpdateMeshOrientation(
    Facing Coordinate Space:ENiagaraCoordinateSpace=::World
    Facing Direction=$Particles.Direction
    Facing Interpolation=false
    Facing Interpolation Rate=0.5
    Orientation Method=::NewEnumerator1
    Override Quat=false
    TransformFacing=true
    TransformUp=false
   )
   /SetVariables(Particles.Debug_Gnomon_Scale:Vector3f=<5 5 5>) {enabled=false, isSetParameters=true}
   &Light_Attributes(
    Color=1
    Color Mode=0
    Falloff Exponent=1
    Light Color:LinearColor=&MakeLinearColorFromVectorAndFloat(
     "Vector (RGB)"=&MakeVectorFromLinearColorRGB(Color=<rgba 3 0.968 0.7832 1>)
     "Float (Alpha)"=&Waveform("Waveform [1]"=::NewEnumerator3, "[1] Frequency"=2, "[1] Amplitude Min/Max"=<0.7 1>, "[1] Scale"=25)
    )
    Light Exponent=8
    Light Radius=200
    Radius=1
   )
   /SetVariables(Particles.PointLightIntensity:NiagaraFloat=&Lerp_Float(
     A=0
     B=&Waveform("Waveform [1]"=::NewEnumerator3, "[1] Frequency"=2, "[1] Amplitude Min/Max"=<0.7 1>, "[1] Scale"=150)
     Alpha=&FloatFromCurve(
      FloatCurve=DataInterface() {propertyValues {Curve {
         defaultValue=1
         keys [<at 0.3 0 RCIM_Cubic>, {arriveTangent=3.5848, interpMode=RCIM_Cubic, leaveTangent=3.5848, tangentMode=RCTM_User, time=1, value=1}]
        }}}
      CurveIndex=$System.Timeline.NormalizedTime
     )
    )) {isSetParameters=true}
  ]
  rendererClasses [
   &NiagaraMeshRendererProperties
   &NiagaraLightRendererProperties
   &NiagaraComponentRendererProperties
  ]
  renderers [
   {rendererClass=&NiagaraMeshRendererProperties, rendererIndex=0}
   {rendererClass=&NiagaraMeshRendererProperties, rendererIndex=1}
   {rendererClass=&NiagaraLightRendererProperties, rendererIndex=2}
   {rendererClass=&NiagaraComponentRendererProperties, rendererIndex=3}
  ]
 }
 Corona001 {
  emitterUpdateScript [
   &EmitterState(
    Delay First Loop Only=true
    Inactive Response=::NewEnumerator1
    Life Cycle Mode=::NewEnumerator1
    Loop Delay=$System.EmitterDelay
    Loop Duration=1
    Recalculate Duration Each Loop=true
    UseLoopDelay=true
   )
   &SpawnBurst_Instantaneous(Use Spawn Probability=false, UseLoopCountLimit=false)
  ]
  particleSpawnScript [
   &InitializeParticle(
    Color=<rgba 2 0.4 0.4267 1>
    Lifetime=1.5
    Sprite Size Mode=::NewEnumerator3
    Uniform Sprite Size=75
    UsePositionOffset=false
    Write Material Random=true
   )
   &CameraOffset(Camera Offset Amount=75)
   &DynamicMaterialParameters(
    Index 0 Param 1=1
    Index 0 Param 2=0.2
    Index 0 Param 3=0.5
    Index 0 Param 4=1
    Module.Index0Float0Write=true
    Module.Index0Float1Write=true
    Module.Index0Float2Write=true
    Module.Index0Float3Write=true
   )
  ]
  particleUpdateScript [
   &ParticleState()
   &ScaleColor(
    Scale Alpha=&FloatFromCurve(FloatCurve=DataInterface() {propertyValues {Curve {
        defaultValue=1
        keys [
         {interpMode=RCIM_Cubic, tangentMode=RCTM_User, time=0, value=0}
         {arriveTangentWeight=0.0833, interpMode=RCIM_Cubic, leaveTangentWeight=0.1667, tangentMode=RCTM_User, time=0.6, value=1}
         {tangentMode=RCTM_User, time=1, value=0}
        ]
       }}})
    ScaleA=true
    ScaleRGB=false
   )
   &self:Follow_Pos_Simple(
    Fallback Position:NiagaraPosition=$User.RocketPos
    Particle Attribute Reader=DataInterface() {propertyValues {EmitterBinding {bindingMode=Other, emitterName=Bug_Wings}}}
   ) {module=Follow_Pos_Simple}
  ]
  rendererClasses [&NiagaraSpriteRendererProperties]
  renderers [{rendererClass=&NiagaraSpriteRendererProperties}]
 }
 Dark_Gradient {
  emitterUpdateScript [
   &EmitterState(
    Delay First Loop Only=true
    Inactive Response=::NewEnumerator1
    Life Cycle Mode=::NewEnumerator1
    Loop Delay=$System.EmitterDelay
    Loop Duration=2
    Recalculate Duration Each Loop=true
    UseLoopDelay=true
   )
   &SpawnBurst_Instantaneous(Use Spawn Probability=false, UseLoopCountLimit=false)
  ]
  particleSpawnScript [
   &InitializeParticle(
    Color=<rgba 0.0175 0.0243 0.03 0.6>
    Lifetime=3.25
    Sprite Size Mode=::NewEnumerator3
    Uniform Sprite Size=150
    UsePositionOffset=false
    Write Material Random=true
   )
   &CameraOffset(Camera Offset Amount=-50)
   &DynamicMaterialParameters(
    Index 0 Param 1=3
    Index 0 Param 2=1.5
    Module.Index0Float0Write=true
    Module.Index0Float1Write=true
    Module.Index0Float2Write=false
    Module.Index0Float3Write=false
   )
  ]
  particleUpdateScript [
   &ParticleState()
   &ScaleColor(
    Scale Alpha=&FloatFromCurve(FloatCurve=DataInterface() {propertyValues {Curve {
        defaultValue=1
        keys [
         {interpMode=RCIM_Cubic, tangentMode=RCTM_User, time=0, value=0}
         {arriveTangentWeight=0.0833, interpMode=RCIM_Cubic, leaveTangentWeight=0.1667, tangentMode=RCTM_User, time=0.25, value=1}
         {arriveTangentWeight=0.2778, interpMode=RCIM_Cubic, leaveTangentWeight=0.1389, tangentMode=RCTM_User, time=0.75, value=1}
         {tangentMode=RCTM_User, time=1, value=0}
        ]
       }}})
    ScaleA=true
    ScaleRGB=false
   )
   &self:Follow_Pos_Simple(
    Fallback Position:NiagaraPosition=$User.RocketPos
    Particle Attribute Reader=DataInterface() {propertyValues {EmitterBinding {bindingMode=Other, emitterName=Bug_Wings}}}
   ) {module=Follow_Pos_Simple}
  ]
  rendererClasses [&NiagaraSpriteRendererProperties]
  renderers [{rendererClass=&NiagaraSpriteRendererProperties}]
 }
 Embers {
  emitterUpdateScript [
   &EmitterState(Life Cycle Mode=::NewEnumerator1, Loop Behavior=::NewEnumerator1, Loop Delay=1.85, Loop Duration=25, UseLoopDelay=true)
   &SpawnBurst_Instantaneous(Spawn Count=50, Use Spawn Probability=false, UseLoopCountLimit=true)
   &SpawnRate(
    SpawnRate=&Multiply_Float(
     A=25
     B=&FloatFromCurve(FloatCurve=DataInterface() {propertyValues {Curve {keys [<at 0 1 RCIM_Constant>,<at 12.85 8 RCIM_Constant>]}}}, CurveIndex=$Emitter.Age)
    )
    Use Spawn Probability=false
   )
  ]
  particleSpawnScript [
   &InitializeParticle(
    Color=<rgba 3 0.15 0.15 1>
    Lifetime Max=7
    Lifetime Min=2
    Lifetime Mode=::NewEnumerator1
    Sprite Rotation Mode=::NewEnumerator2
    Sprite Size Mode=::NewEnumerator1
    Uniform Sprite Size Max=2
    Uniform Sprite Size Min=1
    UsePositionOffset=false
    Write Material Random=true
   )
   &self:Follow_Pos_Simple(
    Fallback Position:NiagaraPosition=$User.RocketPos
    Needs World Space Transform=true
    Particle Attribute Reader=DataInterface() {propertyValues {EmitterBinding {bindingMode=Other, emitterName=Bug_Wings}}}
   ) {module=Follow_Pos_Simple}
   &ShapeLocation(
    MaskSpawnGroup=false
    Override Seed=false
    Sphere Radius=&RandomRangeFloat(Minimum=2, Maximum=10)
    Sphere Surface Distribution=1
    UseOwnerScale=true
   )
   &AddVelocity(Velocity=&RandomRangeVector(Minimum=<0 0 -5>, Maximum=<0 0 -12>))
  ]
  particleUpdateScript [
   &ParticleState()
   &ScaleSpriteSize(Uniform Curve Sprite Scale=DataInterface() {propertyValues {Curve {keys [
        <at 0 0 RCIM_Cubic>
        {arriveTangentWeight=0.0738, interpMode=RCIM_Cubic, leaveTangentWeight=0.2569, tangentMode=RCTM_User, time=0.157, value=1}
        {arriveTangent=-1.7046, interpMode=RCIM_Cubic, leaveTangent=-1.7046, tangentMode=RCTM_User, time=0.7034, value=0.7}
        <at 1 0 RCIM_Cubic>
       ]}}})
   &GravityForce(Gravity=<0 0 -20>)
   /Niagara/VectorFields/SampleVectorField(
    FieldRotate:Quat4f=&MakeQuatFromEuler(Euler:Vector3f=<0 0 -90>, Scale:NiagaraFloat=1)
    VectorField=DataInterface() {propertyValues {Field=/Game/Effects/VectorFields/Noisy_Turbulence_103, bTileX=true, bTileY=true, bTileZ=true}}
   )
   /Niagara/VectorFields/ApplyVectorField(Vector Field Intensity as Force=25, Vector Field Intensity as Velocity=0.25, Vector Field Velocity Tightness=0.01)
   &SolveForcesAndVelocity(Clamp Velocity=false, Limit Acceleration=false)
   &ScaleColor(
    Scale RGB=&VectorFromCurve(VectorCurve=DataInterface() {propertyValues {XCurve {keys [<at 0 6000>,<at 1 5000>]}, YCurve {keys [<at 0 6000>,<at 1 1140.9097>]}, ZCurve {keys [<at 0 6000>,<at 1 4886.832>]}}})
    ScaleA=true
    ScaleRGB=true
   )
  ]
  rendererClasses [&NiagaraSpriteRendererProperties]
  renderers [{rendererClass=&NiagaraSpriteRendererProperties}]
 }
 Flare_Loop {
  emitterUpdateScript [
   &EmitterState(
    Delay First Loop Only=true
    Inactive Response=::NewEnumerator1
    Life Cycle Mode=::NewEnumerator1
    Loop Delay=$System.EmitterDelay
    Loop Duration=1
    Recalculate Duration Each Loop=true
    UseLoopDelay=true
   )
   &SpawnBurst_Instantaneous(Use Spawn Probability=false, UseLoopCountLimit=false)
  ]
  particleSpawnScript [
   &InitializeParticle(
    Color=<rgba 3 0.3 0.345 1>
    Lifetime=1.5
    Sprite Size Mode=::NewEnumerator3
    Uniform Sprite Size=125
    UsePositionOffset=false
    Write Material Random=true
   )
   &CameraOffset(Camera Offset Amount=50)
   &DynamicMaterialParameters(
    Index 0 Param 1=0.5
    Index 0 Param 2=1
    Module.Index0Float0Write=true
    Module.Index0Float1Write=true
    Module.Index0Float2Write=false
    Module.Index0Float3Write=false
   )
  ]
  particleUpdateScript [
   &ParticleState()
   &ScaleColor(
    Scale Alpha=&FloatFromCurve(FloatCurve=DataInterface() {propertyValues {Curve {
        defaultValue=1
        keys [
         {interpMode=RCIM_Cubic, tangentMode=RCTM_User, time=0, value=0}
         {arriveTangentWeight=0.0833, interpMode=RCIM_Cubic, leaveTangentWeight=0.1667, tangentMode=RCTM_User, time=0.6, value=1}
         {tangentMode=RCTM_User, time=1, value=0}
        ]
       }}})
    ScaleA=true
    ScaleRGB=false
   )
   &self:Follow_Pos_Simple(
    Fallback Position:NiagaraPosition=$User.RocketPos
    Particle Attribute Reader=DataInterface() {propertyValues {EmitterBinding {bindingMode=Other, emitterName=Bug_Wings}}}
   ) {module=Follow_Pos_Simple}
  ]
  rendererClasses [&NiagaraSpriteRendererProperties]
  renderers [{rendererClass=&NiagaraSpriteRendererProperties}]
 }
 Ribbon {
  emitterUpdateScript [
   &EmitterState(Life Cycle Mode=::NewEnumerator1, Loop Behavior=::NewEnumerator1, Loop Delay=12.5, Loop Duration=25, UseLoopDelay=true)
   &SpawnRate(SpawnRate=10, Use Spawn Probability=false)
  ]
  particleSpawnScript [
   &InitializeParticle(
    Color=<rgba 1 0.1 0.115 0.2>
    Lifetime=7
    Ribbon Width=25
    Ribbon Width Mode=::NewEnumerator1
    UsePositionOffset=false
    Write Material Random=true
   )
   &self:Follow_Pos_Simple(
    Fallback Position:NiagaraPosition=$User.RocketPos
    Needs World Space Transform=true
    Particle Attribute Reader=DataInterface() {propertyValues {EmitterBinding {bindingMode=Other, emitterName=Bug_Wings}}}
   ) {module=Follow_Pos_Simple}
  ]
  particleUpdateScript [
   &ParticleState()
   &DynamicMaterialParameters(
    Index 0 Param 1=&FloatFromCurve(FloatCurve=DataInterface() {propertyValues {Curve {keys [
         <at 0 1 RCIM_Cubic>
         {arriveTangent=-3.4151, arriveTangentWeight=2.2251, interpMode=RCIM_Cubic, leaveTangent=-3.4151, tangentMode=RCTM_User, time=1, value=-1}
        ]}}})
    Module.Index0Float0Write=true
    Module.Index0Float1Write=false
    Module.Index0Float2Write=false
    Module.Index0Float3Write=false
   )
   &GravityForce(Gravity=<0 0 -20>)
   &SolveForcesAndVelocity(Clamp Velocity=false, Limit Acceleration=false)
  ]
  rendererClasses [&NiagaraRibbonRendererProperties]
  renderers [{rendererClass=&NiagaraRibbonRendererProperties}]
 }
}
summary {
 systemName=NS_Geode_AlteredBug
 emitters [
  {
   emitterName=Bug_Wings
   rendererClasses [
    &NiagaraMeshRendererProperties
    &NiagaraLightRendererProperties
    &NiagaraComponentRendererProperties
   ]
  }
  {emitterName=Embers, rendererClasses [&NiagaraSpriteRendererProperties]}
  {emitterName=Bug_Body, rendererClasses [&NiagaraSpriteRendererProperties]}
  {emitterName=Dark_Gradient, rendererClasses [&NiagaraSpriteRendererProperties]}
  {emitterName=Flare_Loop, rendererClasses [&NiagaraSpriteRendererProperties]}
  {emitterName=Corona001, rendererClasses [&NiagaraSpriteRendererProperties]}
  {emitterName=Ribbon, rendererClasses [&NiagaraRibbonRendererProperties]}
 ]
 userVariables [User.RocketPos {type=NiagaraPosition, description="", defaultValue {}}]
}
