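# Niagara contact effect for the Geode "AlteredBug", showing a staged crack with drifting fragments, glints and embers.
# Crack spawns camera-offset meshes in three timed bursts, animated through dynamic material parameters.
# TinyFrags mixes bursts with steady spawning and seeds Glints sprites through particle reads, while a delayed Embers burst swirls on curl noise.
# A system LODScale variable scales sizes and spawn radii; flare, corona and gradient emitters list no renderers.
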NiagaraSystem /Figment_S04_Geode/Athena/Effects/Particles/Geode/NS_Geode_AlteredBug_Contact
dependencies {
 usedDataInterfaces [&NiagaraDataInterfaceCurve, &NiagaraDataInterfaceParticleRead, &NiagaraDataInterfaceVectorCurve]
 usedDynamicInputs [
  &FN_LOD_Distance_Remap
  &Vector2DFromFloat
  &Multiply_Vector2D_ByFloat
 ]
 usedModules [
  &SystemState
  &self:SystemSpawnScript.SystemScriptSource.SystemScriptGraph.NiagaraNodeAssignment_0.SetVariables_#1
  &self:SystemSpawnScript.SystemScriptSource.SystemScriptGraph.NiagaraNodeAssignment_1.SetVariables_#2
 ]
}
emitters {
 Corona {rendererClasses [], renderers []}
 Corona_Quick {rendererClasses [], renderers []}
 Crack {
  emitterUpdateScript [
   &EmitterState()
   &SpawnBurst_Instantaneous(Use Spawn Probability=false, UseLoopCountLimit=false)
   &SpawnBurst_Instantaneous(Spawn Time=0.4, Use Spawn Probability=false, UseLoopCountLimit=false) {module=SpawnBurst_Instantaneous001}
   &SpawnBurst_Instantaneous(Spawn Time=0.9, Use Spawn Probability=false, UseLoopCountLimit=false) {module=SpawnBurst_Instantaneous002}
  ]
  particleSpawnScript [
   &InitializeParticle(
    Color Mode=::NewEnumerator0
    Lifetime=&FloatFromCurve(FloatCurve=DataInterface() {propertyValues {Curve {keys [<at 0 1.85>,<at 1 0>]}}}, CurveIndex=$Emitter.NormalizedLoopAge)
    Mesh Scale Mode=::NewEnumerator3
    Mesh Uniform Scale=2
    UsePositionOffset=false
    Write Material Random=true
   )
   &InitialMeshOrientation(Use Rotation Vector=false)
   &CameraOffset(Camera Offset Amount=500)
  ]
  particleUpdateScript [
   &ParticleState()
   &ScaleMeshSize(Scale Factor=&VectorFromFloat(Value=&FloatFromCurve(
      FloatCurve=DataInterface() {propertyValues {Curve {keys [{arriveTangent=0.3616, interpMode=RCIM_Cubic, leaveTangent=0.3616, tangentMode=RCTM_User, time=0, value=1}, <at 1 2 RCIM_Cubic>]}}}
      Scale Curve=$System.LODScale
     )))
   &DynamicMaterialParameters(
    Index 0 Param 1=&FloatFromCurve(
     FloatCurve=DataInterface() {propertyValues {Curve {keys [
         {arriveTangent=1.432, interpMode=RCIM_Cubic, leaveTangent=1.432, tangentMode=RCTM_User, time=0, value=0}
         {interpMode=RCIM_Cubic, leaveTangent=1.3217, time=1, value=1}
        ]}}}
     Scale Curve=&FloatFromCurve(
      FloatCurve=DataInterface() {propertyValues {Curve {keys [<at 0.8 1 RCIM_Cubic>, {arriveTangent=-9.463, interpMode=RCIM_Cubic, leaveTangent=-9.463, tangentMode=RCTM_User, time=1, value=0}]}}}
      CurveIndex=$Emitter.NormalizedLoopAge
     )
    )
    Index 0 Param 2=&FloatFromCurve(
     FloatCurve=DataInterface() {propertyValues {Curve {keys [{arriveTangent=1.6343, interpMode=RCIM_Cubic, leaveTangent=1.6343, tangentMode=RCTM_User, time=0, value=0}, <at 1 1 RCIM_Cubic>]}}}
     Scale Curve=&Multiply_Float(
      A=&Multiply_Float(A=0.5, B=$System.LODScale)
      B=&FloatFromCurve(
       FloatCurve=DataInterface() {propertyValues {Curve {keys [<at 0.8 1 RCIM_Cubic>, {arriveTangent=-9.463, interpMode=RCIM_Cubic, leaveTangent=-9.463, tangentMode=RCTM_User, time=1, value=0}]}}}
       CurveIndex=$Emitter.NormalizedLoopAge
      )
     )
    )
    Index 0 Param 3=&FloatFromCurve(
     FloatCurve=DataInterface() {propertyValues {Curve {keys [<at 0 -0.15>, {arriveTangent=0.55, interpMode=RCIM_Cubic, leaveTangent=0.55, time=0.7, value=-0.018}, <at 1 0.4 RCIM_Cubic>]}}}
     Scale Curve=0.125
    )
    Module.Index0Float0Write=true
    Module.Index0Float1Write=true
    Module.Index0Float2Write=true
    Module.Index0Float3Write=true
   )
  ]
  rendererClasses [&NiagaraMeshRendererProperties]
  renderers [{rendererClass=&NiagaraMeshRendererProperties, rendererIndex=0}]
 }
 Dark_Gradient {rendererClasses [], renderers []}
 Embers {
  emitterUpdateScript [
   &EmitterState()
   &SpawnBurst_Instantaneous(Spawn Count=50, Spawn Time=1.5, Use Spawn Probability=false, UseLoopCountLimit=true)
  ]
  particleSpawnScript [
   &InitializeParticle(
    Color=<rgba 3 0.075 0.075 1>
    Lifetime Max=3
    Lifetime Mode=::NewEnumerator1
    Sprite Rotation Mode=::NewEnumerator2
    Sprite Size Mode=::NewEnumerator1
    Uniform Sprite Size Max=2
    Uniform Sprite Size Min=1
    UsePositionOffset=false
    Write Material Random=true
   )
   &ShapeLocation(
    MaskSpawnGroup=false
    Override Seed=false
    Sphere Radius=&Multiply_Float(A=&RandomRangeFloat(Minimum=2, Maximum=10), B=$System.LODScale)
    Sphere Surface Distribution=1
    UseOwnerScale=true
   )
   &AddVelocity(Module.Use Falloff Distance=false, Override Seed=false, Velocity Mode=::NewEnumerator1, Velocity Speed=75)
  ]
  particleUpdateScript [
   &ParticleState()
   &ScaleSpriteSize(
    Uniform Curve Scale=$System.LODScale
    Uniform Curve Sprite Scale=DataInterface() {propertyValues {Curve {keys [
        <at 0 0 RCIM_Cubic>
        {arriveTangentWeight=0.0738, interpMode=RCIM_Cubic, leaveTangentWeight=0.2569, tangentMode=RCTM_User, time=0.157, value=1}
        {arriveTangent=-1.7046, interpMode=RCIM_Cubic, leaveTangent=-1.7046, tangentMode=RCTM_User, time=0.7034, value=0.7}
        <at 1 0 RCIM_Cubic>
       ]}}}
   )
   &CurlNoiseForce(Bias Noise=false, Noise Frequency=150, Noise Strength=500, Pan Noise=true, Pan Noise Field=<0 0 1>)
   &GravityForce(Gravity=&RandomRangeVector(Minimum=<0 0 -20>, Maximum=<0 0 -100>)) {enabled=false}
   &SolveForcesAndVelocity(Clamp Velocity=false, Limit Acceleration=false)
   &ScaleColor(
    Scale RGB=&VectorFromCurve(VectorCurve=DataInterface() {propertyValues {XCurve {keys [<at 0 6000>,<at 1 5000>]}, YCurve {keys [<at 0 6000>,<at 1 1140.9097>]}, ZCurve {keys [<at 0 6000>,<at 1 4886.832>]}}})
    ScaleA=true
    ScaleRGB=true
   )
  ]
  rendererClasses [&NiagaraSpriteRendererProperties]
  renderers [{rendererClass=&NiagaraSpriteRendererProperties}]
 }
 Glints {
  emitterUpdateScript [
   &EmitterState()
   &SpawnParticlesFromOtherEmitter(
    Attribute Reader=DataInterface() {propertyValues {EmitterBinding {bindingMode=Other, emitterName=TinyFrags}}}
    Calculate Spawn Rate Per Particle=false
    Spawn Rate=20
    Use Maximum Age=false
    Use Minimum Age=false
   )
  ]
  particleSpawnScript [
   &InitializeParticle(
    Color=<rgba 1 0.2 0.2 1>
    Lifetime=&RandomRangeFloat(Minimum=0.12, Maximum=0.6)
    Sprite Size Mode=::NewEnumerator1
    Uniform Sprite Size Max=125
    Uniform Sprite Size Min=75
    UsePositionOffset=false
    Write Material Random=true
   )
   &SampleParticlesFromOtherEmitter(Velocity Sampling=::NewEnumerator0)
  ]
  particleUpdateScript [
   &ParticleState()
   &ScaleSpriteSize(
    Uniform Curve Scale=$System.LODScale
    Uniform Curve Sprite Scale=DataInterface() {propertyValues {Curve {keys [
        {arriveTangent=0.4699, interpMode=RCIM_Cubic, leaveTangent=0.4699, tangentMode=RCTM_User, time=0, value=0}
        {arriveTangent=0.9942, interpMode=RCIM_Cubic, leaveTangent=0.9942, tangentMode=RCTM_User, time=0.575, value=0.349}
        {arriveTangent=-1.8214, interpMode=RCIM_Cubic, leaveTangent=-1.8214, tangentMode=RCTM_User, time=0.7, value=1}
        <at 1 0 RCIM_Cubic>
       ]}}}
   )
   &SolveForcesAndVelocity(Clamp Velocity=false, Limit Acceleration=false)
  ]
  rendererClasses [&NiagaraSpriteRendererProperties]
  renderers [{rendererClass=&NiagaraSpriteRendererProperties}]
 }
 Round_Flare {rendererClasses [], renderers []}
 Sharp_Flare {rendererClasses [], renderers []}
 TinyFrags {
  emitterUpdateScript [
   &EmitterState()
   &SpawnBurst_Instantaneous(Spawn Count=2, Spawn Time=0.3, Use Spawn Probability=false, UseLoopCountLimit=false)
   &SpawnBurst_Instantaneous(Spawn Count=4, Spawn Time=0.45, Use Spawn Probability=false, UseLoopCountLimit=false) {module=SpawnBurst_Instantaneous001}
   &SpawnBurst_Instantaneous(Spawn Count=4, Spawn Time=1, Use Spawn Probability=false, UseLoopCountLimit=false) {module=SpawnBurst_Instantaneous002}
   &SpawnRate(SpawnRate=10, Use Spawn Probability=false)
  ]
  particleSpawnScript [
   &InitializeParticle(
    Color=<rgba 10 5 5.0833 1>
    Lifetime Min=1.5
    Lifetime Mode=::NewEnumerator1
    Mesh Scale Mode=::NewEnumerator1
    Mesh Uniform Scale Max=1.7
    Mesh Uniform Scale Min=0.8
    UsePositionOffset=false
    Write Material Random=true
   )
   &ShapeLocation(
    MaskSpawnGroup=false
    Override Seed=false
    Sphere Radius=&Multiply_Float(A=&RandomRangeFloat(Minimum=20, Maximum=40), B=$System.LODScale)
    UseOwnerScale=true
   )
   &InitialMeshOrientation(Use Rotation Vector=false)
   &AddVelocity(Velocity=&RandomRangeVector(Minimum=<-8 -8 -8>, Maximum=<8 8 8>))
   &AddVelocity(
    Module.Use Falloff Distance=false
    Override Seed=false
    Velocity Mode=::NewEnumerator1
    Velocity Speed=&RandomRangeFloat(Minimum=25, Maximum=100)
   ) {module=AddVelocity001}
  ]
  particleUpdateScript [
   &ParticleState()
   &ScaleMeshSize(Scale Factor=&VectorFromFloat(Value=&FloatFromCurve(
      FloatCurve=DataInterface() {propertyValues {Curve {keys [<at 0 0 RCIM_Cubic>, <at 0.2 1>, {arriveTangent=-1.25, interpMode=RCIM_Cubic, leaveTangent=-1.25, time=0.8, value=1}, <at 1 0>]}}}
      Scale Curve=$System.LODScale
     )))
   &UpdateMeshOrientation(Rotation Vector=&RandomRangeVector(Minimum=<-0.25 -0.15 -0.15>, Maximum=<0.25 0.15 0.15>))
   &Drag(Use Linear Drag=true, Use Rotational Drag=true)
   &SolveForcesAndVelocity(Clamp Velocity=false, Limit Acceleration=false)
  ]
  rendererClasses [&NiagaraMeshRendererProperties]
  renderers [{rendererClass=&NiagaraMeshRendererProperties, rendererIndex=0}]
 }
}
summary {
 systemName=NS_Geode_AlteredBug_Contact
 emitters [
  {emitterName=Dark_Gradient, rendererClasses []}
  {emitterName=Crack, rendererClasses [&NiagaraMeshRendererProperties]}
  {emitterName=Round_Flare, rendererClasses []}
  {emitterName=Sharp_Flare, rendererClasses []}
  {emitterName=Corona, rendererClasses []}
  {emitterName=Corona_Quick, rendererClasses []}
  {emitterName=TinyFrags, rendererClasses [&NiagaraMeshRendererProperties]}
  {emitterName=Glints, rendererClasses [&NiagaraSpriteRendererProperties]}
  {emitterName=Embers, rendererClasses [&NiagaraSpriteRendererProperties]}
 ]
 userVariables []
}
