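# Niagara spawn effect for the Geode Altered Bug: shrinking mesh fragments, red glints and a pulsing red light.
# TinyFrags spawns three tapering groups in sphere and hemisphere shapes, pulls them toward an offset center, and sets a visibility tag per group across two mesh renderers.
# Glints spawns sprites from TinyFrags particles via attribute reader, while Light bursts and fades by curve.
# The gradient, beam and flare emitters have no renderers.
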NiagaraSystem /Figment_S04_Geode/Athena/Effects/Particles/Geode/NS_Geode_AlteredBug_Spawn
dependencies {
 usedDataInterfaces [&NiagaraDataInterfaceCurve, &NiagaraDataInterfaceArrayInt32, &NiagaraDataInterfaceParticleRead]
 usedDynamicInputs []
 usedModules [
  &self:SystemSpawnScript.SystemScriptSource.SystemScriptGraph.NiagaraNodeAssignment_0.SetVariables_#1
  &SystemState
 ]
}
emitters {
 Dark_Gradient {rendererClasses [], renderers []}
 FlareWide {rendererClasses [], renderers []}
 Flare_Start {rendererClasses [], renderers []}
 Glints {
  emitterUpdateScript [
   &EmitterState(Life Cycle Mode=::NewEnumerator1, Loop Behavior=::NewEnumerator1, Loop Duration=1, UseLoopDelay=false)
   &SpawnParticlesFromOtherEmitter(
    Attribute Reader=DataInterface() {propertyValues {EmitterBinding {bindingMode=Other, emitterName=TinyFrags}}}
    Calculate Spawn Rate Per Particle=false
    Spawn Rate=&FloatFromCurve(
     FloatCurve=DataInterface() {propertyValues {Curve {keys [
         {arriveTangent=3.6088, interpMode=RCIM_Cubic, leaveTangent=3.6088, tangentMode=RCTM_User, time=5, value=6}
         {arriveTangent=12.228, interpMode=RCIM_Cubic, leaveTangent=12.228, tangentMode=RCTM_User, time=8, value=24}
         {arriveTangent=-13.3534, interpMode=RCIM_Cubic, leaveTangent=-13.3534, tangentMode=RCTM_User, time=9, value=40}
         {interpMode=RCIM_Cubic, leaveTangent=6.2102, time=10, value=5}
        ]}}}
     CurveIndex=$Emitter.Age
     Scale Curve=3
    )
    Use Maximum Age=false
    Use Minimum Age=false
   )
  ]
  particleSpawnScript [
   &InitializeParticle(
    Color=<rgba 1 0.2 0.2 1>
    Lifetime=&RandomRangeFloat(Minimum=0.1, Maximum=0.6)
    Sprite Size Mode=::NewEnumerator1
    Uniform Sprite Size Max=400
    Uniform Sprite Size Min=100
    UsePositionOffset=false
    Write Material Random=true
   )
   &SampleParticlesFromOtherEmitter(Emitter Sampling Mode:ENiagara_EmitterLocSamplingMode=::NewEnumerator0, Velocity Sampling=::NewEnumerator0)
  ]
  particleUpdateScript [
   &ParticleState()
   &ScaleMeshSize(Scale Factor=&VectorFromFloat(Value=&FloatFromCurve(FloatCurve=DataInterface() {propertyValues {Curve {keys [
          {arriveTangent=0.4699, interpMode=RCIM_Cubic, leaveTangent=0.4699, tangentMode=RCTM_User, time=0, value=0}
          {arriveTangent=0.9942, interpMode=RCIM_Cubic, leaveTangent=0.9942, tangentMode=RCTM_User, time=0.575, value=0.349}
          {arriveTangent=-1.8214, interpMode=RCIM_Cubic, leaveTangent=-1.8214, tangentMode=RCTM_User, time=0.7, value=1}
          <at 1 0 RCIM_Cubic>
         ]}}})))
   &SolveForcesAndVelocity(Clamp Velocity=false, Limit Acceleration=false)
  ]
  rendererClasses [&NiagaraSpriteRendererProperties]
  renderers [{rendererClass=&NiagaraSpriteRendererProperties}]
 }
 Light {
  emitterUpdateScript [
   &EmitterState()
   &SpawnBurst_Instantaneous(Use Spawn Probability=false, UseLoopCountLimit=true)
  ]
  particleSpawnScript [&InitializeParticle(
    Color Mode=::NewEnumerator0
    Lifetime=2
    Position Offset:NiagaraPosition=$System.PositionOffset
    UsePositionOffset=true
    Write Material Random=true
   )]
  particleUpdateScript [
   &ParticleState()
   &Light_Attributes(
    Color=1
    Color Mode=0
    Falloff Exponent=1
    Light Color:LinearColor=&MakeLinearColorFromVectorAndFloat(
     "Vector (RGB)"=&MakeVectorFromLinearColorRGB(Color=<rgba 3 0.075 0.075 1>)
     "Float (Alpha)"=&FloatFromCurve(
      FloatCurve=DataInterface() {propertyValues {Curve {
         defaultValue=1
         keys [
          {arriveTangent=7.8215, interpMode=RCIM_Cubic, leaveTangent=7.8215, tangentMode=RCTM_User, time=0, value=0}
          {arriveTangentWeight=0.162, interpMode=RCIM_Cubic, leaveTangentWeight=0.2268, tangentMode=RCTM_User, time=0.25, value=1}
          {arriveTangentWeight=0.1944, interpMode=RCIM_Cubic, leaveTangentWeight=0.2222, tangentMode=RCTM_User, time=0.6, value=1}
          <at 1 0 RCIM_Cubic>
         ]
        }}}
      Scale Curve=10
     )
    )
    Light Exponent=3
    Light Radius=400
    Radius=1
   )
  ]
  rendererClasses [&NiagaraLightRendererProperties]
  renderers [{rendererClass=&NiagaraLightRendererProperties, rendererIndex=0}]
 }
 LightBeams {rendererClasses [], renderers []}
 TinyFrags {
  emitterUpdateScript [
   &EmitterState(Life Cycle Mode=::NewEnumerator1, Loop Behavior=::NewEnumerator1, Loop Duration=1, UseLoopDelay=false)
   &SpawnRate(
    SpawnRate=&FloatFromCurve(
     FloatCurve=DataInterface() {propertyValues {Curve {keys [<at 0 1 RCIM_Cubic>, {interpMode=RCIM_Cubic, leaveTangent=-3.8275, time=0.5, value=0}]}}}
     CurveIndex=$Emitter.NormalizedLoopAge
     Scale Curve=20
    )
    Use Spawn Probability=false
   )
   &SpawnRate(
    SpawnGroup=1
    SpawnRate=&FloatFromCurve(
     FloatCurve=DataInterface() {propertyValues {Curve {keys [<at 0 1 RCIM_Cubic>, {interpMode=RCIM_Cubic, leaveTangent=-3.5266, time=0.5, value=0}]}}}
     CurveIndex=$Emitter.NormalizedLoopAge
     Scale Curve=20
    )
    Use Spawn Probability=false
   ) {module=SpawnRate001}
   &SpawnRate(
    SpawnGroup=2
    SpawnRate=&FloatFromCurve(
     FloatCurve=DataInterface() {propertyValues {Curve {keys [<at 0 1 RCIM_Cubic>, {interpMode=RCIM_Cubic, leaveTangent=-3.1328, time=0.5, value=0}]}}}
     CurveIndex=$Emitter.NormalizedLoopAge
     Scale Curve=15
    )
    Use Spawn Probability=false
   ) {module=SpawnRate002}
  ]
  particleSpawnScript [
   /SetVariables(Particles.VisibilityTag=&SelectIntFromArray(Direct Array Index=&SpawnGroup(), Int Selection Array=DataInterface() {propertyValues {IntData [0,0,1]}})) {isSetParameters=true}
   &InitializeParticle(
    Color Mode=::NewEnumerator0
    Lifetime=&FloatFromCurve(
     FloatCurve=DataInterface() {propertyValues {Curve {keys [{arriveTangent=-4.1974, interpMode=RCIM_Cubic, leaveTangent=-4.1974, tangentMode=RCTM_User, time=0, value=1}, <at 0.5 0.1 RCIM_Cubic>]}}}
     CurveIndex=$Emitter.NormalizedLoopAge
     Scale Curve=1.5
    )
    Mesh Scale Mode=::NewEnumerator3
    Mesh Uniform Scale=&MaskFloatBySpawnGroup(
     Masked Float=&RandomRangeFloat(Minimum=2, Maximum=7)
     Passthrough Float=&MaskFloatBySpawnGroup(
      Masked Float=&RandomRangeFloat(Minimum=4, Maximum=8)
      Passthrough Float=&RandomRangeFloat(Minimum=3, Maximum=5)
     )
    )
    Position Offset:NiagaraPosition=$System.PositionOffset
    UsePositionOffset=true
    Write Material Random=true
   )
   &ShapeLocation(
    MaskSpawnGroup=true
    Override Seed=false
    Sphere Radius=&Lerp_Float(
     A=&RandomRangeFloat(Minimum=125, Maximum=200)
     B=&RandomRangeFloat(Minimum=50, Maximum=200)
     Alpha=&FloatFromCurve(FloatCurve=DataInterface() {propertyValues {Curve {keys [<at 0.5 0>,<at 1 1>]}}}, CurveIndex=$Emitter.NormalizedLoopAge)
    )
    Sphere Surface Distribution=1
    UseOwnerScale=true
   )
   &ShapeLocation(
    Hemisphere Distribution=<1 0.5>
    MaskSpawnGroup=true
    Override Seed=false
    Rotation Axis=<0 0 -1>
    Rotation Mode=::NewEnumerator1
    Spawn Group Mask=1
    Sphere Radius=&RandomRangeFloat(Minimum=150, Maximum=200)
    Sphere Surface Distribution=1
    UseOwnerScale=true
   ) {module=ShapeLocation001}
   &ShapeLocation(
    MaskSpawnGroup=true
    Override Seed=false
    Spawn Group Mask=2
    Sphere Radius=&Lerp_Float(
     A=&RandomRangeFloat(Minimum=125, Maximum=200)
     B=&RandomRangeFloat(Minimum=50, Maximum=200)
     Alpha=&FloatFromCurve(FloatCurve=DataInterface() {propertyValues {Curve {keys [<at 0.5 0>,<at 1 1>]}}}, CurveIndex=$Emitter.NormalizedLoopAge)
    )
    Sphere Surface Distribution=1
    UseOwnerScale=true
   ) {module=ShapeLocation002}
   &InitialMeshOrientation(Use Rotation Vector=false)
   &AddVelocity(Velocity=&RandomRangeVector(Minimum=<-8 -8 -8>, Maximum=<8 8 8>))
  ]
  particleUpdateScript [
   &ParticleState()
   &UpdateMeshOrientation(
    Rotation Rate=2
    Rotation Vector=&RandomRangeVector(Minimum=<-0.15 -0.15 -0.15>, Maximum=<0.15 0.15 0.15>)
   )
   &ScaleMeshSize(Scale Factor=&VectorFromFloat(Value=&FloatFromCurve(FloatCurve=DataInterface() {propertyValues {Curve {keys [<at 0.5 1>,<at 1 0>]}}})))
   &PointAttractionForce(
    Attraction Radius=4000
    Attraction Strength=25
    Attractor Position Offset:NiagaraPosition=$System.PositionOffset
    Kill Within Radius=false
    Module.Use Falloff=false
    UseAttractorOffset=true
   )
   &ScaleVelocity(Velocity Scale=&VectorFromFloat(Value=&FloatFromCurve(FloatCurve=DataInterface() {propertyValues {Curve {keys [<at 0.5 1>,<at 1 0>]}}})))
   &Drag(Drag=0.2, Use Linear Drag=true, Use Rotational Drag=true)
   &SolveForcesAndVelocity(Clamp Velocity=false, Limit Acceleration=false)
   &ScaleColor(
    Scale Alpha=&FloatFromCurve(FloatCurve=DataInterface() {propertyValues {Curve {keys [<at 0 0 RCIM_Cubic>,<at 0.7 1 RCIM_Cubic>,<at 1 0>]}}})
    ScaleA=true
    ScaleRGB=true
   )
  ]
  rendererClasses [&NiagaraMeshRendererProperties]
  renderers [
   {rendererClass=&NiagaraMeshRendererProperties, rendererIndex=0}
   {rendererClass=&NiagaraMeshRendererProperties, rendererIndex=1}
  ]
 }
}
summary {
 systemName=NS_Geode_AlteredBug_Spawn
 emitters [
  {emitterName=Dark_Gradient, rendererClasses []}
  {emitterName=LightBeams, rendererClasses []}
  {emitterName=Flare_Start, rendererClasses []}
  {emitterName=FlareWide, rendererClasses []}
  {emitterName=TinyFrags, rendererClasses [&NiagaraMeshRendererProperties]}
  {emitterName=Glints, rendererClasses [&NiagaraSpriteRendererProperties]}
  {emitterName=Light, rendererClasses [&NiagaraLightRendererProperties]}
 ]
 userVariables []
}
